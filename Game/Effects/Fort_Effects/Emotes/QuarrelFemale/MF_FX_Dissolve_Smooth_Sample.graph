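# Material function in the QuarrelFemale emote effects folder that turns a scalar texture sample into a soft-edged dissolve mask, exposed to the library under FX Opacity.
# It multiplies Texture by Smooth, then subtracts a threshold that lerps from Smooth to -1 as Steps rises.
# The saturated Dissolve output grows from empty to full for zero-to-one inputs, with higher Smooth values giving a harder edge.
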MaterialFunction /Game/Effects/Fort_Effects/Emotes/QuarrelFemale/MF_FX_Dissolve_Smooth_Sample
graph {
 properties {bExposeToLibrary=true, libraryCategoriesText [FX Opacity]}
 nodes [
  n1 = FunctionInput() {bUsePreviewValueAsDefault=true, inputName=Texture, inputType=FunctionInput_Scalar, previewValue=<1 0 0 0>}
  n2 = FunctionInput() {bUsePreviewValueAsDefault=true, inputName=Smooth, inputType=FunctionInput_Scalar, sortPriority=1, previewValue=<3 0 0 0>}
  n3 = Multiply(A=@n1, B=@n2)
  n4 = FunctionInput() {bUsePreviewValueAsDefault=true, inputName=Steps, inputType=FunctionInput_Scalar, sortPriority=2, previewValue=<1 0 0 0>}
  n5 = LinearInterpolate(A=@n2, Alpha=@n4) {constB=-1}
  FunctionOutput(None=Saturate(None=Subtract(A=@n3, B=@n5))) {outputName=Dissolve, sortPriority=1}
 ]
}
