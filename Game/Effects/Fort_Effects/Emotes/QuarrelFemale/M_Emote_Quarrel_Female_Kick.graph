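# Additive, unlit, two-sided Niagara mesh-particle material for the kick effects in the female Quarrel emote.
# Scalar parameters blend three looks: a curve-colored flash with panning tornado detail, a distorted scrolling-smoke helix trail, and a WPO smoke trail.
# Both trails fade at glancing angles and use texture-driven dissolves.
# Dynamic parameters vary tiling, dissolve and timing, and drive the WPO trail's sine/cosine swirl plus noise-based vertex displacement, weighted along U.
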Material /Game/Effects/Fort_Effects/Emotes/QuarrelFemale/M_Emote_Quarrel_Female_Kick
graph {
 properties {
  bUsedWithNiagaraMeshParticles=true
  blendMode=BLEND_Additive
  pixelDepthOffsetMode=PDOM_Legacy
  shadingModel=MSM_Unlit
  twoSided=true
  usedShadingModels=Unlit
 }
 nodes [
  n1 = TextureCoordinate() {vTiling=0.2}
  n2 = DynamicParameter() {parameterIndex=2, paramNames [WPO_UV_Offset_Random, WPO_Strength_Random, WPO_UV_Time_Random, WPO_Strength_Curve]}
  n3 = Multiply(A=Time(), B=@n2.WPO_UV_Time_Random)
  n4 = TextureSample(UVs=Add(A=Panner(Coordinate=@n1, Time=@n3) {speedX=-0.2}, B=@n2.WPO_UV_Offset_Random)) {samplerType=SAMPLERTYPE_LinearColor, texture=&HighContrastNoise}
  ComponentMask(None=@n4.RGB) {b=true, g=true}
  n5 = DynamicParameter() {parameterIndex=2, paramNames [WPO_UV_Offset_Random, WPO_Strength_Random, WPO_UV_Time_Random, WPO_Strength_Curve]}
  n6 = ComponentMask(None=TextureCoordinate()) {r=true}
  n7 = DynamicParameter() {parameterIndex=2, paramNames [WPO_UV_Offset_Random, WPO_Strength_Random, WPO_UV_Time_Random, WPO_Strength_Curve]}
  n8 = Add(A=Add(A=Multiply(A=@n6, B=@n7.WPO_UV_Time_Random), B=@n7.WPO_UV_Time_Random), B=@n7.WPO_UV_Offset_Random)
  n9 = Constant3Vector() {constant=<rgba 0 1 0 0>}
  n10 = Add(A=Add(A=Multiply(A=@n6, B=@n7.WPO_UV_Time_Random), B=@n7.WPO_UV_Time_Random), B=@n7.WPO_UV_Offset_Random)
  n11 = Add(A=Multiply(A=VertexNormalWS(), B=Cosine(None=@n8)), B=Multiply(A=Transform(None=@n9), B=Sine(None=@n10)))
  n12 = TextureCoordinate()
  n13 = Saturate(None=Multiply(A=ComponentMask(None=@n12) {r=true}) {constB=3})
  n14 = Multiply(A=Multiply(A=@n5.WPO_Strength_Random, B=Multiply(A=@n11) {constB=10}), B=Saturate(None=Add(A=@n13) {constB=0.1})) {constB=20}
  n15 = Multiply(A=ComponentMask(None=TextureCoordinate()) {r=true}) {constB=2}
  n16 = AppendVector(A=Constant(), B=AppendVector(A=Multiply(A=@n4.G) {constB=50}, B=Multiply(A=@n4.B) {constB=50}))
  n17 = DynamicParameter() {parameterIndex=2, paramNames [WPO_UV_Offset_Random, WPO_Strength_Random, WPO_UV_Time_Random, WPO_Strength_Curve]}
  n18 = Transform(None=Multiply(A=Saturate(None=@n15), B=Multiply(A=@n16, B=@n17.WPO_Strength_Random)) {constB=20})
  LinearInterpolate(
   B=Add(A=Multiply(A=@n5.WPO_Strength_Curve, B=@n14) {constB=20}, B=Multiply(A=@n18, B=@n17.WPO_Strength_Curve))
   Alpha=ScalarParameter() {parameterName=Use WPOTrail}
  )
  n19 = ParticleColor()
  n20 = TextureSample(UVs=TextureCoordinate()) {samplerType=SAMPLERTYPE_LinearGrayscale, texture=&/T_Emote_Quarrel_Female_Kick_Flash_01}
  n21 = TextureCoordinate() {uTiling=3, vTiling=2}
  n22 = DynamicParameter() {defaultValue=<rgba 0 1 1 1>, paramNames [TextureBlend,PanSpeed,TilingRandom,Dissolve]}
  n23 = Multiply(A=@n22.PanSpeed, B=Time())
  n24 = Multiply(
   A=@n20.R
   B=TextureSample(UVs=Panner(Coordinate=@n21, Time=@n23) {speedY=1}) {texture=/Game/Global/Textures/Default/Tileable_Patterns/T_Tornado_Tile}.R
  )
  n25 = Multiply(A=@n24, B=@n24)
  n26 = LinearInterpolate(A=@n20.R, B=Multiply(A=@n25, B=@n25), Alpha=@n22.TextureBlend)
  n27 = ParticleColor()
  n28 = MaterialFunctionCall(Float2=TextureCoordinate()) {materialFunction=&BreakOutFloat2Components}
  n29 = AppendVector(
   A=Multiply(A=@n28.R, B=DynamicParameter() {defaultValue=<rgba 0 1 1 1>, paramNames [TextureBlend,PanSpeed,TilingRandom,Dissolve]}.TilingRandom)
   B=@n28.G
  )
  n30 = Panner(Coordinate=@n29) {speedX=-3}
  n31 = Panner(Coordinate=Multiply(
    A=TextureCoordinate()
    B=DynamicParameter() {defaultValue=<rgba 0 1 1 1>, paramNames [TextureBlend,PanSpeed,TilingRandom,Dissolve]}.TilingRandom
   )) {speedX=-1}
  n32 = Add(A=Multiply(A=TextureSample(UVs=@n31) {samplerType=SAMPLERTYPE_LinearGrayscale, texture=/Game/Global/Textures/Default/Tileable_Patterns/T_StormCloudDetail01_M}.R) {constB=0.2}) {constB=-0.1}
  n33 = Add(A=Panner(Coordinate=@n29) {speedX=-2}, B=@n32)
  n34 = Add(
   A=TextureSample(UVs=Add(A=@n30, B=@n32)) {samplerType=SAMPLERTYPE_Grayscale, texture=&T_Smoke_Trail_Soft}.R
   B=Multiply(A=TextureSample(UVs=@n33) {texture=/Game/Effects/Fort_Effects/Textures/T_InkySmokeTrail}.R) {constB=3}
  )
  n35 = VectorParameter() {group=HelixTrail, parameterName=HelixTrail, channelNames {a="", b="", g=Mask Mul, r=Emis Add}, defaultValue=<rgba 0 1 0 0>}
  n36 = Multiply(A=@n34, B=@n35.Mask Mul) {constB=5}
  n37 = MaterialFunctionCall(Float2=TextureCoordinate()) {materialFunction=&BreakOutFloat2Components}
  n38 = CurveAtlasRowParameter(CurveTime=OneMinus(None=@n37.R)) {
   atlas=/Game/Athena/LinearColorCurves/VFX_CurveAtlas
   curve=&/Curve_Emote_Quarrel_Female_Kick
   defaultValue=32
   parameterName=Mask
  }
  n39 = Multiply(A=@n38.A, B=Sine(None=Multiply(A=@n37.G) {constB=0.5})) {constA=1}
  n40 = Abs(None=DotProduct(A=CameraVectorWS(), B=VertexNormalWS()))
  n41 = Saturate(None=Multiply(A=@n40, B=@n40))
  n42 = TextureCoordinate()
  n43 = VectorParameter() {group=Dissolve, parameterName=Dissolve UV, sortPriority=3, defaultValue=<rgba 1 1 0 0>}
  n44 = DynamicParameter() {defaultValue=<rgba 0 1 1 1>, paramNames [TextureBlend,PanSpeed,TilingRandom,Dissolve]}
  n45 = Multiply(
   A=MaterialFunctionCall(X=@n43.R, Y=@n43.R) {materialFunction=&MakeFloat2}.Result
   B=@n44.TilingRandom
  )
  n46 = Multiply(
   A=MaterialFunctionCall(X=@n43.B, Y=@n43.B) {materialFunction=&MakeFloat2}.Result
   B=Time()
  )
  n47 = TextureSample(
   UVs=Add(A=Multiply(A=@n42, B=@n45), B=Add(A=@n46, B=@n44.TilingRandom))
   Tex=TextureObjectParameter() {group=Dissolve, parameterName=Dissolve Tex, sortPriority=2}
  )
  n48 = VectorParameter() {
   group=Dissolve
   parameterName=Dissolve Control
   sortPriority=1
   channelNames {a=Intensity, b=Smooth, g=Tex Multiply, r=Use Position Dissolve}
   defaultValue=<rgba 1 1 3 1>
  }
  n49 = Add(
   A=Multiply(A=@n47.R, B=@n48.Tex Multiply)
   B=MaterialFunctionCall(Float2=TextureCoordinate()) {materialFunction=&BreakOutFloat2Components}.R
  )
  n50 = LinearInterpolate(A=@n47.R, B=Multiply(A=Saturate(None=@n49), B=@n48.Smooth), Alpha=@n48.Use Position Dissolve)
  n51 = LinearInterpolate(
   A=@n48.Smooth
   Alpha=Multiply(A=DynamicParameter() {defaultValue=<rgba 0 1 1 1>, paramNames [TextureBlend,PanSpeed,TilingRandom,Dissolve]}.Dissolve, B=@n48.Intensity)
  ) {constB=-1}
  n52 = Multiply(A=Multiply(A=Multiply(A=@n36, B=@n39) {constA=1}, B=@n41), B=Saturate(None=Subtract(A=@n50, B=@n51)))
  n53 = ScalarParameter() {group=HelixTrail, parameterName=Use HelixTrail}
  n54 = MaterialFunctionCall(Float2=TextureCoordinate() {uTiling=0.5, vTiling=2}) {materialFunction=&BreakOutFloat2Components}
  n55 = AppendVector(
   A=Add(
    A=@n54.R
    B=DynamicParameter() {parameterIndex=2, paramNames [WPO_UV_Offset_Random, WPO_Strength_Random, WPO_UV_Time_Random, WPO_Strength_Curve]}.WPO_UV_Offset_Random
   )
   B=Add(A=@n54.G) {constB=-0.5}
  )
  n56 = Multiply(A=Power(Base=TextureSample(UVs=@n55) {samplerType=SAMPLERTYPE_Grayscale, texture=&T_Smoke_Trail_Soft}.R) {constExponent=5}) {constB=10}
  n57 = MaterialFunctionCall(Float2=TextureCoordinate()) {materialFunction=&BreakOutFloat2Components}
  n58 = AppendVector(A=@n57.G, B=OneMinus(None=@n57.R))
  n59 = TextureSample(UVs=@n58) {texture=/Game/Effects/Fort_Effects/Textures/T_Smoke_Ring_Mask2}
  n60 = TextureSample(UVs=@n58) {texture=/Game/Global/Textures/Default/Masks_and_Gradients/T_Water_Top_Ramp}
  n61 = Add(
   A=TextureCoordinate() {uTiling=2, vTiling=0.5}
   B=DynamicParameter() {parameterIndex=2, paramNames [WPO_UV_Offset_Random, WPO_Strength_Random, WPO_UV_Time_Random, WPO_Strength_Curve]}.WPO_UV_Offset_Random
  )
  n62 = Multiply(
   A=Multiply(A=@n56, B=Multiply(A=@n59.R, B=@n60.R))
   B=MaterialFunctionCall!(
    Texture=TextureSample(UVs=@n61) {texture=&T_Inky_Smoke_Tile}.R
    Smooth=Constant() {r=10}
    Steps=DynamicParameter() {parameterIndex=1, paramNames [Dissolve2,Param2,Param3,Param4]}.Dissolve2
   ) {materialFunction=&/MF_FX_Dissolve_Smooth_Sample}.Dissolve
  )
  n63 = ParticleColor()
  LinearInterpolate(
   A=LinearInterpolate(A=Multiply(A=@n19.A, B=@n26), B=Multiply(A=@n27.A, B=@n52), Alpha=@n53)
   B=Multiply(A=Multiply(A=@n62, B=@n63.A), B=@n41)
   Alpha=ScalarParameter() {parameterName=Use WPOTrail}
  )
  n64 = CurveAtlasRowParameter(CurveTime=@n20.R) {
   atlas=&VFX_CurveAtlas
   curve=&Curve_Emote_Quarrel_Female_Kick
   defaultValue=32
   parameterName=Curve
  }
  n65 = Multiply(A=@n19.RGB, B=@n64)
  n66 = Multiply(A=@n27.RGB, B=Add(A=@n34, B=@n35.Emis Add) {constB=0.1})
  MaterialFunctionCall(Emissive=LinearInterpolate(A=LinearInterpolate(A=@n65, B=@n66, Alpha=@n53), B=Multiply(A=@n56, B=@n63.RGB), Alpha=ScalarParameter() {parameterName=Use WPOTrail})) {materialFunction=&MF_FX_EmissiveAdjustments}
  Sine(None=@n8)
 ]
}
